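# Mono looping MetaSound source for the Cartmanland music party's default track, switching between music and a fireworks feed.
# On play or a change to bStartFireworks, a bool compare sets or resets a value that is interpolated to crossfade the wave player against the Wonkeeland fireworks audio bus.
# Reset plays a random wave from the Waves array, while set stops it through a trigger delay that a reset cancels.
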MetaSoundSource /Hera/Audio/Music/Diegetic/Cartmanland/MSS_CartmanLand_MusicParty_Track_Default
inputs [
 bStartFireworks {type=Bool}
 UE.Source.OnPlay {type=Trigger}
 UE.Source.StartTime.StartTime {type=Time}
 Waves {type=WaveAsset:Array}
]
interfaces [UE.OutputFormat.Mono:1.0, UE.Source.StartTime:1.0, UE.Source:1.1]
outputs ["UE.OutputFormat.Mono.Audio:0" {type=Audio}]
pages [Default {
  input_defaults {UE.Source.OnPlay=null, UE.Source.StartTime.StartTime=null, Waves=null, bStartFireworks=null}
  nodes [
   AudioBusReader1_0 = "UE.Audio Bus Reader (1)"() {also {defaults {Audio Bus=/Hera/Audio/Music/Diegetic/WonkeeLand/AB_Wonkeeland_Fireworks}}}
   n1 = InitVariable.Float()
   TriggerOnValueChange_0 = UE.Trigger On Value Change:Bool(Value=^bStartFireworks)
   TriggerAnyTriggerAny2_0 = "TriggerAny.Trigger Any (2)"(In 1=^UE.Source.OnPlay, In 0=@TriggerOnValueChange_0.Trigger)
   n2 = TriggerCompare.Clamp:Bool(A=UE.Reroute(Value=^bStartFireworks).Value, Compare=@TriggerAnyTriggerAny2_0.Out) {also {defaults {B="<MetaSoundBuilderResult.SUCCEEDED: 0>"}}}
   n3 = Value.Value:Float(Reset=@n2.False, Set=@n2.True) {also {defaults {Init Value=0, Target Value=1}}}
   n4 = UE.InterpTo:Audio(Target=@n3.Output Value) {also {defaults {Interp Time=1}}}
   n5 = UE.Subtract:Float(AdditionalOperands=@n4.Value) {also {defaults {PrimaryOperand=1}}}
   n6 = InitVariable.Float()
   n7 = InitVariable.Trigger()
   ArrayRandomGet_0 = Array.Random Get:WaveAsset:Array(In Array=^Waves, Next=VariableAccessor.Trigger(Variable=VariableMutator.Trigger(Variable=@n7.Variable, Value=@n3.On Reset).Variable).Value)
   n8 = InitVariable.Trigger()
   TriggerDelay_0 = UE.Trigger Delay(Reset=@n3.On Reset, In=@n3.On Set)
   WavePlayer_0 = UE.Wave Player:Mono(
    Start Time=^UE.Source.StartTime.StartTime
    Play=@ArrayRandomGet_0.On Next
    Wave Asset=@ArrayRandomGet_0.Value
    Stop=VariableAccessor.Trigger(Variable=VariableMutator.Trigger(Variable=@n8.Variable, Value=@TriggerDelay_0.Out).Variable).Value
   ) {also {defaults {Loop="<MetaSoundBuilderResult.SUCCEEDED: 0>"}, version [1,1]}}
   AudioMixerAudioMixerMono2_0 = "AudioMixer.Audio Mixer (Mono, 2)"(
    In 1=@AudioBusReader1_0.Out 0
    Gain 0=VariableAccessor.Float(Variable=VariableMutator.Float(Variable=@n1.Variable, Value=@n5.Out).Variable).Value
    Gain 1=VariableAccessor.Float(Variable=VariableMutator.Float(Variable=@n6.Variable, Value=@n4.Value).Variable).Value
    In 0=@WavePlayer_0.Out Mono
   )
  ]
  links [{dst="@UE.OutputFormat.Mono.Audio:0", src=AudioMixerAudioMixerMono2_0.Out}]
 }]
properties {
 assetImportData=&self:AssetImportData
 bEnableBaseSubmix=false
 bEnableSubmixSends=false
 bLooping=true
 bOverrideConcurrency=true
 duration=10000
 sampleRate=48000
 virtualizationMode=PlayWhenSilent
 concurrencyOverrides {maxCount=4}
}
