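# Looping mono MetaSound source for diegetic Cartmanland shooting gallery music, switching between idle, gameplay and fireworks.
# StartGame plays a transition stinger and a delayed gameplay track while fading out the idle loop; StopGame reverses the fade.
# bStartFireworks crossfades idle and gameplay to a Wonkeeland fireworks audio bus, stopping idle after a cancellable delay and restarting it when fireworks end.
# The stinger bypasses these fades, and the Waves input is unused.
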MetaSoundSource /Hera/Audio/Music/Diegetic/Cartmanland/MSS_CartmanLand_MusicParty_Track_Shooting
inputs [
 bStartFireworks {type=Bool}
 StartGame {type=Trigger}
 StopGame {type=Trigger}
 UE.Source.OnPlay {type=Trigger}
 UE.Source.StartTime.StartTime {type=Time}
 Waves {type=WaveAsset:Array}
]
interfaces [UE.OutputFormat.Mono:1.0, UE.Source.StartTime:1.0, UE.Source:1.1]
outputs ["UE.OutputFormat.Mono.Audio:0" {type=Audio}]
pages [Default {
  input_defaults {StartGame=null, StopGame=null, UE.Source.OnPlay=null, UE.Source.StartTime.StartTime=0, Waves=null, bStartFireworks=null}
  nodes [
   AudioBusReader1_0 = "UE.Audio Bus Reader (1)"() {also {defaults {Audio Bus=/Hera/Audio/Music/Diegetic/WonkeeLand/AB_Wonkeeland_Fireworks}}}
   n1 = InitVariable.Float()
   TriggerOnValueChange_0 = UE.Trigger On Value Change:Bool(Value=^bStartFireworks)
   TriggerAnyTriggerAny2_0 = "TriggerAny.Trigger Any (2)"(In 1=^UE.Source.OnPlay, In 0=@TriggerOnValueChange_0.Trigger)
   n2 = TriggerCompare.Clamp:Bool(A=UE.Reroute(Value=^bStartFireworks).Value, Compare=@TriggerAnyTriggerAny2_0.Out) {also {defaults {B="<MetaSoundBuilderResult.SUCCEEDED: 0>"}}}
   n3 = Value.Value:Float(Reset=@n2.False, Set=@n2.True) {also {defaults {Init Value=0, Target Value=1}}}
   n4 = UE.InterpTo:Audio(Target=@n3.Output Value) {also {defaults {Interp Time=1}}}
   n5 = UE.Subtract:Float(AdditionalOperands=@n4.Value) {also {defaults {PrimaryOperand=1}}}
   n6 = UE.InterpTo:Audio(Target=Value.Value:Float(Set=^StartGame, Reset=^StopGame) {also {defaults {Init Value=0, Target Value=1}}}.Output Value) {also {defaults {Interp Time=1.5}}}
   n7 = UE.Multiply:Float(AdditionalOperands=@n5.Out, PrimaryOperand=UE.Subtract:Float(AdditionalOperands=@n6.Value) {also {defaults {PrimaryOperand=1}}}.Out)
   n8 = InitVariable.Float()
   n9 = InitVariable.Float()
   n10 = UE.Multiply:Float(AdditionalOperands=@n6.Value, PrimaryOperand=@n5.Out)
   n11 = InitVariable.Trigger()
   n12 = InitVariable.Trigger()
   TriggerDelay_0 = UE.Trigger Delay(Reset=@n3.On Reset, In=@n3.On Set)
   WavePlayer_0 = UE.Wave Player:Mono(
    Start Time=^UE.Source.StartTime.StartTime
    Play=VariableAccessor.Trigger(Variable=VariableMutator.Trigger(Variable=@n11.Variable, Value=@n3.On Reset).Variable).Value
    Stop=VariableAccessor.Trigger(Variable=VariableMutator.Trigger(Variable=@n12.Variable, Value=@TriggerDelay_0.Out).Variable).Value
   ) {also {
     defaults {
      Loop="<MetaSoundBuilderResult.SUCCEEDED: 0>"
      Wave Asset=&/MX_S39_IP_HoneyFiber_Wonkeeland_ShooterGallery_Idle_LP
     }
     version [1,1]
    }}
   WavePlayer_1 = UE.Wave Player:Mono(Play=^StartGame) {also {defaults {Wave Asset=&/MX_S39_IP_HoneyFiber_Wonkeeland_ShooterGallery_Transition_STNG}, version [1,1]}}
   TriggerDelay_1 = UE.Trigger Delay(In=^StartGame) {also {defaults {Delay Time=0.827}}}
   WavePlayer_2 = UE.Wave Player:Mono(Play=@TriggerDelay_1.Out) {also {defaults {Wave Asset=&/MX_S39_IP_HoneyFiber_Wonkeeland_ShooterGallery_Gameplay_STNG}, version [1,1]}}
   AudioMixerAudioMixerMono4_0 = "AudioMixer.Audio Mixer (Mono, 4)"(
    In 3=@AudioBusReader1_0.Out 0
    Gain 2=VariableAccessor.Float(Variable=VariableMutator.Float(Variable=@n1.Variable, Value=@n7.Out).Variable).Value
    Gain 3=VariableAccessor.Float(Variable=VariableMutator.Float(Variable=@n8.Variable, Value=@n4.Value).Variable).Value
    Gain 0=VariableAccessor.Float(Variable=VariableMutator.Float(Variable=@n9.Variable, Value=@n10.Out).Variable).Value
    In 2=@WavePlayer_0.Out Mono
    In 1=@WavePlayer_1.Out Mono
    In 0=@WavePlayer_2.Out Mono
   )
  ]
  links [{dst="@UE.OutputFormat.Mono.Audio:0", src=AudioMixerAudioMixerMono4_0.Out}]
 }]
properties {
 assetImportData=&self:AssetImportData
 bEnableBaseSubmix=false
 bEnableSubmixSends=false
 bLooping=true
 bOverrideConcurrency=true
 duration=10000
 sampleRate=48000
 virtualizationMode=PlayWhenSilent
 volume=1.5
 assetUserData [&self:AudioAssetUserData_0]
 concurrencyOverrides {maxCount=4}
}
